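# Sound cue for readying the SupersonicPink cosmetic pickaxe, played in the melee weapon sound class.
# A mixer layers one of six random pickaxe swing waves with a quieter-scaled, briefly delayed random pick from three supersonic vehicle start waves.
# Each layer passes through its own modulator at default settings.
# A concurrency override caps simultaneous instances, stopping the oldest, and attenuation reuses the player pickaxe impact settings.
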SoundCue /BRCosmetics/Sounds/PickAxes/SupersonicPink/Pickaxe_Ready_SupersonicPink
root=Mixer_1
nodes [
 {id=Delay_0, children [Modulator_0], properties {delayMax=0.1, delayMin=0.1}}
 {id=Mixer_1, children [Delay_0,Modulator_1], properties {inputVolume [0.25,0.15]}}
 {id=Modulator_0, children [Random_1]}
 {id=Modulator_1, children [Random_0]}
 {id=Random_0, children [WavePlayer_3,WavePlayer_4,WavePlayer_5,WavePlayer_6,WavePlayer_7,WavePlayer_8], properties {weights [1,1,1,1,1,1]}}
 {id=Random_1, children [WavePlayer_0,WavePlayer_1,WavePlayer_2], properties {weights [1,1,1]}}
 {id=WavePlayer_0, properties {soundWaveAssetPtr=&/BC2_Vehicle_SuperSonic_Start_03}}
 {id=WavePlayer_1, properties {soundWaveAssetPtr=&/BC2_Vehicle_SuperSonic_Start_01}}
 {id=WavePlayer_2, properties {soundWaveAssetPtr=&/BC2_Vehicle_SuperSonic_Start_02}}
 {id=WavePlayer_3, properties {soundWaveAssetPtr=&/Pickaxe_SupersonicPink_Swing_01}}
 {id=WavePlayer_4, properties {soundWaveAssetPtr=&/Pickaxe_SupersonicPink_Swing_02}}
 {id=WavePlayer_5, properties {soundWaveAssetPtr=&/Pickaxe_SupersonicPink_Swing_03}}
 {id=WavePlayer_6, properties {soundWaveAssetPtr=&/Pickaxe_SupersonicPink_Swing_04}}
 {id=WavePlayer_7, properties {soundWaveAssetPtr=&/Pickaxe_SupersonicPink_Swing_05}}
 {id=WavePlayer_8, properties {soundWaveAssetPtr=&/Pickaxe_SupersonicPink_Swing_06}}
]
properties {
 attenuationSettings=/Game/Sounds/Attenuation/Fort_Weapon_Pickaxe_Player_Impact_Att
 audioPropertiesSheet=/BRCosmetics/Sounds/Settings/PropertySheets/Pickaxe/APS_Cosmetic_Pickaxe_Ready
 bOverrideConcurrency=true
 maxDistance=3512
 soundClassObject=&Fort_WeaponMelee
 volumeMultiplier=1
 concurrencyOverrides {maxCount=2, resolutionRule=StopOldest}
}
